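# Physics asset for the InflateMate microphone instrument cosmetic, simulating its balloon tie and dangling ribbon.
# A capsule handle body anchors a sphere on the balloon tie, which leads a chain of four ribbon capsules.
# Constraints lock twist along the whole chain, holding the tie to a tight swing on the handle while the ribbon segments swing much more freely.
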PhysicsAsset /SparksCosmetics/Instrument/Mic/InflateMate/Meshes/InflateMate_Mic_Physics
bodies {
 dyn_balloon_tie [{shapeType=Sphere, radius=1.1041, shapeName=dyn_balloon_tie_sphere, center=<1 0 0>}]
 dyn_ribbon_01 [{shapeType=Capsule, length=2.3236, radius=0.6, shapeName=dyn_ribbon_01_capsule, center=<2 0 0>, rotation=<rot -90 0 0>}]
 dyn_ribbon_02 [{shapeType=Capsule, length=1.5556, radius=0.6, shapeName=dyn_ribbon_02_capsule, center=<2 0 0>, rotation=<rot -90 0 0>}]
 dyn_ribbon_03 [{shapeType=Capsule, length=2.6218, radius=0.6, shapeName=dyn_ribbon_03_capsule, center=<2 0 0>, rotation=<rot -90 0 0>}]
 dyn_ribbon_04 [{shapeType=Capsule, length=3.7373, radius=0.7056, shapeName=dyn_ribbon_04_capsule, center=<2 0 0>, rotation=<rot -90 0 0>}]
 handle [{shapeType=Capsule, length=10, radius=3, center=<0 0 -5.5205>}]
}
constraints [
 {bone1Name=dyn_ribbon_01, bone2Name=dyn_balloon_tie, swing1LimitDegrees=60, swing2LimitDegrees=45, twistLimitDegrees=45, twistMotion=Locked}
 {bone1Name=dyn_balloon_tie, bone2Name=handle, swing1LimitDegrees=4, swing2LimitDegrees=4, twistLimitDegrees=45, twistMotion=Locked}
 {bone1Name=dyn_ribbon_04, bone2Name=dyn_ribbon_03, swing1LimitDegrees=45, swing2LimitDegrees=45, twistLimitDegrees=45, twistMotion=Locked}
 {bone1Name=dyn_ribbon_02, bone2Name=dyn_ribbon_01, swing1LimitDegrees=45, swing2LimitDegrees=45, twistLimitDegrees=45, twistMotion=Locked}
 {bone1Name=dyn_ribbon_03, bone2Name=dyn_ribbon_02, swing1LimitDegrees=45, swing2LimitDegrees=45, twistLimitDegrees=45, twistMotion=Locked}
]
